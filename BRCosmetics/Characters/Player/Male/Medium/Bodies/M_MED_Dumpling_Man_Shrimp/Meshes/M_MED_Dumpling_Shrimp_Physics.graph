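# Physics asset for the male medium Dumpling Man Shrimp body, defining bodies and joints for its asymmetric dynamic jacket.
# Capsules cover thighs, calves and feet, with small spheres at pelvis, left thigh twist and right jacket leg master.
# Right side and rear four-box chains hang from the thigh and leg master, tightly cross-linked at their third segments.
# A shorter two-box left rear chain hangs from the left thigh.
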PhysicsAsset /BRCosmetics/Characters/Player/Male/Medium/Bodies/M_MED_Dumpling_Man_Shrimp/Meshes/M_MED_Dumpling_Shrimp_Physics
bodies {
 calf_l [{shapeType=Capsule, length=32.8591, radius=7.9516, center=<-20.107 -1.2538 0.0804>, rotation=<rot 0 92.386 -88.9929>}]
 calf_r [{shapeType=Capsule, length=31.1136, radius=9.1024, center=<19.477 0.9667 0.0048>, rotation=<rot 0 90.4706 -88.4774>}]
 dyn_jacket_leg_master_r [{shapeType=Sphere, radius=0.5}]
 dyn_jacket_rear_01_l [{shapeType=Box, extentX=11.3377, extentY=4.0404, extentZ=12.2133, center=<5.3514 0 0>}]
 dyn_jacket_rear_01_r [{shapeType=Box, extentX=10.5, extentY=4.7214, extentZ=19.5936, center=<5.5097 0.0001 0>}]
 dyn_jacket_rear_02_l [{shapeType=Box, extentX=11.01, extentY=4.01, extentZ=13.01, center=<4.582 0 0>}]
 dyn_jacket_rear_02_r [{shapeType=Box, extentX=16.51, extentY=7.01, extentZ=23, center=<8.266 0 0>}]
 dyn_jacket_rear_03_r [{shapeType=Box, extentX=14.7495, extentY=6.7987, extentZ=22.7557, center=<7.1801 0 0>}]
 dyn_jacket_rear_04_r [{shapeType=Box, extentX=7.01, extentY=7.01, extentZ=24.01, center=<3.3525 0 0>}]
 dyn_jacket_side_01_r [{shapeType=Box, extentX=11.4001, extentY=3.5, extentZ=11.0485, center=<5.5958 0 0>}]
 dyn_jacket_side_02_r [{shapeType=Box, extentX=14.239, extentY=4, extentZ=21.9487, center=<6.7253 -0.0001 0.0001>}]
 dyn_jacket_side_03_r [{shapeType=Box, extentX=15.01, extentY=4, extentZ=22.01, center=<6.1171 -0.0002 -0.0003>}]
 dyn_jacket_side_04_r [{shapeType=Box, extentX=8.6559, extentY=3.9838, extentZ=22.2534, center=<3.0209 0 -0.0002>}]
 foot_l [{shapeType=Capsule, length=13.0269, radius=8.5852, center=<-2.0659 -0.2309 0.0487>, rotation=<rot 0 12.9291 -87.6292>}]
 foot_r [{shapeType=Capsule, length=12.6312, radius=14.2382, center=<-1.0902 -4.0847 3.8303>, rotation=<rot 0 15.0072 -87.5416>}]
 pelvis [{shapeType=Sphere, radius=5}]
 thigh_l [{shapeType=Capsule, length=29.9368, radius=10.2129, center=<-23.7747 -0.2131 0.3068>, rotation=<rot 90 0 0>}]
 thigh_r [{shapeType=Capsule, length=31.1969, radius=10.1765, center=<28 0.3685 -0.3333>, rotation=<rot 90 0 0>}]
 thigh_twist_01_l [{shapeType=Sphere, radius=2}]
}
constraints [
 {bone1Name=dyn_jacket_side_01_r, bone2Name=thigh_r, swing1LimitDegrees=45, swing2LimitDegrees=45, twistLimitDegrees=5}
 {bone1Name=dyn_jacket_side_02_r, bone2Name=dyn_jacket_side_01_r, swing1LimitDegrees=45, swing2LimitDegrees=25, twistLimitDegrees=5}
 {bone1Name=dyn_jacket_side_03_r, bone2Name=dyn_jacket_side_02_r, swing1LimitDegrees=45, swing2LimitDegrees=20, twistLimitDegrees=5}
 {bone1Name=dyn_jacket_side_04_r, bone2Name=dyn_jacket_side_03_r, swing1LimitDegrees=45, swing2LimitDegrees=20, twistLimitDegrees=5}
 {bone1Name=dyn_jacket_rear_01_r, bone2Name=dyn_jacket_leg_master_r, swing1LimitDegrees=45, swing2LimitDegrees=20, twistLimitDegrees=5}
 {bone1Name=dyn_jacket_rear_02_r, bone2Name=dyn_jacket_rear_01_r, swing1LimitDegrees=45, swing2LimitDegrees=20, twistLimitDegrees=5}
 {bone1Name=dyn_jacket_rear_03_r, bone2Name=dyn_jacket_rear_02_r, swing1LimitDegrees=45, swing2LimitDegrees=20, twistLimitDegrees=5}
 {bone1Name=dyn_jacket_rear_04_r, bone2Name=dyn_jacket_rear_03_r, swing1LimitDegrees=45, swing2LimitDegrees=20, twistLimitDegrees=5}
 {bone1Name=dyn_jacket_rear_01_l, bone2Name=thigh_l, swing1LimitDegrees=20, swing2LimitDegrees=20, twistLimitDegrees=3}
 {bone1Name=dyn_jacket_rear_02_l, bone2Name=dyn_jacket_rear_01_l, swing1LimitDegrees=20, swing2LimitDegrees=20, twistLimitDegrees=3}
 {bone1Name=dyn_jacket_side_03_r, bone2Name=dyn_jacket_rear_03_r, swing1LimitDegrees=5, swing2LimitDegrees=5, twistLimitDegrees=3}
]
