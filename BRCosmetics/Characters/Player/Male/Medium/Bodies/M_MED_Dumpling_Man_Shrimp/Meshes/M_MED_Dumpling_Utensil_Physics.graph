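# Physics asset for the utensil on the Dumpling Man Shrimp character body, apparently for secondary motion of a hanging spatula.
# A box body on dyn_spatula is constrained to a small pelvis sphere, with tight swing and wider twist limits.
# Capsules on both thighs, calves and feet have no constraints, so they likely serve as leg colliders for the spatula.
# The right thigh capsule is noticeably fatter than the left.
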PhysicsAsset /BRCosmetics/Characters/Player/Male/Medium/Bodies/M_MED_Dumpling_Man_Shrimp/Meshes/M_MED_Dumpling_Utensil_Physics
bodies {
 calf_l [{shapeType=Capsule, length=32.8591, radius=7.9516, center=<-20.107 -1.2538 0.0804>, rotation=<rot 0 92.386 -88.9929>}]
 calf_r [{shapeType=Capsule, length=32.3636, radius=7.8524, center=<19.477 0.9667 0.0048>, rotation=<rot 0 90.4706 -88.4774>}]
 dyn_spatula [{shapeType=Box, extentX=3.5, extentY=8.5, extentZ=12.5, center=<-0.0003 2.6218 -23.1315>, rotation=<rot 0 0 -5.0055>}]
 foot_l [{shapeType=Capsule, length=13.0269, radius=8.5852, center=<-2.0659 -0.2309 0.0487>, rotation=<rot 0 12.9291 -87.6292>}]
 foot_r [{shapeType=Capsule, length=13.1312, radius=8.7382, center=<2.057 0.1834 -0.0579>, rotation=<rot 0 15.0052 -87.5417>}]
 pelvis [{shapeType=Sphere, radius=5}]
 thigh_l [{shapeType=Capsule, length=25.1868, radius=10.2129, center=<-28 -0.2131 0.3068>, rotation=<rot 90 0 0>}]
 thigh_r [{shapeType=Capsule, length=25.4469, radius=15.9265, center=<28 0.3685 -0.3333>, rotation=<rot 90 0 0>}]
}
constraints [{bone1Name=dyn_spatula, bone2Name=pelvis, swing1LimitDegrees=10, swing2LimitDegrees=10, twistLimitDegrees=60}]
