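# Post-body AnimBP for the Dumpling Man Shrimp cosmetic, copying the parent mesh pose and adding secondary motion to outfit and props.
# A pose driver corrects right-shoulder straps, then skydiving-aware RigidBody simulates, swapped for calf-following jacket constraints when physics is off or LOD exceeds 1.
# AnimDynamics, constraints and look-at animate pots, tassels, knife and spatula, with crouch-driven pot and pouch offsets.
# Skydiving, parachute, crouch-end and targeting events trigger dynamics resets.
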AnimBlueprint /BRCosmetics/Characters/Player/Male/Medium/Bodies/M_MED_Dumpling_Man_Shrimp/Meshes/M_MED_Dumpling_Man_Shrimp_AnimBP
target_skeleton=&/M_MED_Dumpling_Man_Shrimp_Skeleton
graphs [
 AnimGraph {nodes [
   RigidBody() {
    bOverrideWorldGravity=true
    overridePhysicsAsset=&/M_MED_Dumpling_Utensil_Physics
    simulationSpace=BaseBoneSpace
    alphaBoolBlend {blendOption=Linear}
    baseBoneRef {boneName=pelvis}
    componentLinearAccScale=<5 5 5>
    componentLinearVelScale=<5 5 5>
    overrideWorldGravity {z=-980}
   }
   n1 = "Transform (Modify) Bone - Bone: dyn_pouch_Back"!(ComponentPose=Local To Component(LocalPose=Use cached pose 'dynCache'()), Alpha=1, Translation="0,0,0", Rotation="0,-45,-30") -> (Pose) {
    rotationMode=BMM_Additive
    rotationSpace=BCS_BoneSpace
    translationMode=BMM_Additive
    translationSpace=BCS_BoneSpace
    alphaBoolBlend {blendOption=Linear}
    boneToModify {boneName=dyn_pouch_Back}
    rotation {roll=-30, yaw=-45}
   }
   Output Pose(Result=Blend Poses by bool(BlendPose_0=Component To Local(ComponentPose=@n1), BlendPose_1=Use cached pose 'dynCache'(), bActiveValue=Get bIsCrouching()) {blendType=Linear}) {layerGroup=None}
   n2 = Pose Driver - Source: clavicle_r!(SourcePose:Pose Link Structure=Copy Pose From Mesh() {bUseAttachedParent=true}) -> (Pose:Pose Link Structure) {
    is=title AnimGraphNode_PoseDriver
    poseAsset=&/M_MED_Dumpling_Man_Shrimp_Grounded_RBF_PoseAsset
    onlyDriveBones [
     {boneName=dyn_strap_corrective_r_bottom}
     {boneName=dyn_strap_corrective_r_back}
     {boneName=dyn_strap_corrective_top}
     {boneName=dyn_strap_corrective_r}
    ]
    poseTargets [
     {
      bApplyCustomCurve=false
      bIsHidden=false
      distanceMethod=DefaultMethod
      drivenName=Pose_0
      functionType=DefaultFunction
      targetScale=0.2938
      boneTransforms [
       {targetRotation=<rot -85.1915 41.4375 143.0619>, targetTranslation=<4.3575 -0.1466 2.3223>}
       {targetRotation=<rot -64.7854 -3.089 6.0959>, targetTranslation=<-15.5111 0 0>}
      ]
      customCurve {defaultValue=3.4028234663852886e+38, postInfinityExtrap=RCCE_Constant, preInfinityExtrap=RCCE_Constant, keys []}
      targetRotation=<rot 0 0 0>
     }
     {
      bApplyCustomCurve=false
      bIsHidden=false
      distanceMethod=DefaultMethod
      drivenName=Pose_1
      functionType=DefaultFunction
      targetScale=0.5605
      boneTransforms [
       {targetRotation=<rot -64.2747 -120.6598 -83.241>, targetTranslation=<4.3575 -0.1466 2.3223>}
       {targetRotation=<rot -23.3709 86.7554 -36.3331>, targetTranslation=<-15.5111 0 0>}
      ]
      customCurve {defaultValue=3.4028234663852886e+38, postInfinityExtrap=RCCE_Constant, preInfinityExtrap=RCCE_Constant, keys []}
      targetRotation=<rot 0 0 0>
     }
     {
      bApplyCustomCurve=false
      bIsHidden=false
      distanceMethod=DefaultMethod
      drivenName=Pose_2
      functionType=DefaultFunction
      targetScale=0.6181
      boneTransforms [
       {targetRotation=<rot -69.838 165.5965 22.6769>, targetTranslation=<4.3575 -0.1466 2.3223>}
       {targetRotation=<rot -75.1088 43.0245 -78.222>, targetTranslation=<-15.5111 0 0>}
      ]
      customCurve {defaultValue=3.4028234663852886e+38, postInfinityExtrap=RCCE_Constant, preInfinityExtrap=RCCE_Constant, keys []}
      targetRotation=<rot 0 0 0>
     }
     {
      bApplyCustomCurve=false
      bIsHidden=false
      distanceMethod=DefaultMethod
      drivenName=Pose_3
      functionType=DefaultFunction
      targetScale=0.3174
      boneTransforms [
       {targetRotation=<rot -73.4134 -176.9703 4.3592>, targetTranslation=<4.3575 -0.1466 2.3223>}
       {targetRotation=<rot -75.5657 105.7208 -93.8077>, targetTranslation=<-15.5111 0 0>}
      ]
      customCurve {defaultValue=3.4028234663852886e+38, postInfinityExtrap=RCCE_Constant, preInfinityExtrap=RCCE_Constant, keys []}
      targetRotation=<rot 0 0 0>
     }
     {
      bApplyCustomCurve=false
      bIsHidden=false
      distanceMethod=DefaultMethod
      drivenName=Pose_4
      functionType=DefaultFunction
      targetScale=0.2938
      boneTransforms [
       {targetRotation=<rot -87.7995 76.2682 109.0761>, targetTranslation=<4.3575 -0.1466 2.3223>}
       {targetRotation=<rot -69.7589 51.1037 -35.2693>, targetTranslation=<-15.5111 0 0>}
      ]
      customCurve {defaultValue=3.4028234663852886e+38, postInfinityExtrap=RCCE_Constant, preInfinityExtrap=RCCE_Constant, keys []}
      targetRotation=<rot 0 0 0>
     }
     {
      bApplyCustomCurve=false
      bIsHidden=false
      distanceMethod=DefaultMethod
      drivenName=Pose_5
      functionType=DefaultFunction
      targetScale=0.2327
      boneTransforms [
       {targetRotation=<rot -76.2754 172.1506 9.5506>, targetTranslation=<4.3575 -0.1466 2.3223>}
       {targetRotation=<rot -45.0185 104.3596 -56.1379>, targetTranslation=<-15.5111 0 0>}
      ]
      customCurve {defaultValue=3.4028234663852886e+38, postInfinityExtrap=RCCE_Constant, preInfinityExtrap=RCCE_Constant, keys []}
      targetRotation=<rot 0 0 0>
     }
     {
      bApplyCustomCurve=false
      bIsHidden=false
      distanceMethod=DefaultMethod
      drivenName=Pose_6
      functionType=DefaultFunction
      targetScale=0.2327
      boneTransforms [
       {targetRotation=<rot -74.8686 -132.7057 -46.9771>, targetTranslation=<4.3575 -0.1466 2.3223>}
       {targetRotation=<rot -39.6825 113.2505 -70.71>, targetTranslation=<-15.5111 0 0>}
      ]
      customCurve {defaultValue=3.4028234663852886e+38, postInfinityExtrap=RCCE_Constant, preInfinityExtrap=RCCE_Constant, keys []}
      targetRotation=<rot 0 0 0>
     }
     {
      bApplyCustomCurve=false
      bIsHidden=false
      distanceMethod=DefaultMethod
      drivenName=Pose_7
      functionType=DefaultFunction
      targetScale=0.4039
      boneTransforms [
       {targetRotation=<rot -57.8419 -161.5812 -18.7948>, targetTranslation=<4.3575 -0.1466 2.3223>}
       {targetRotation=<rot -66.8684 120.4314 -71.1722>, targetTranslation=<-15.5111 0 0>}
      ]
      customCurve {defaultValue=3.4028234663852886e+38, postInfinityExtrap=RCCE_Constant, preInfinityExtrap=RCCE_Constant, keys []}
      targetRotation=<rot 0 0 0>
     }
     {
      bApplyCustomCurve=false
      bIsHidden=false
      distanceMethod=DefaultMethod
      drivenName=Pose_8
      functionType=DefaultFunction
      targetScale=0.6101
      boneTransforms [
       {targetRotation=<rot -53.6465 -152.7178 -29.1383>, targetTranslation=<4.3575 -0.1466 2.3223>}
       {targetRotation=<rot -17.8084 26.8504 -97.253>, targetTranslation=<-15.5111 0 0>}
      ]
      customCurve {defaultValue=3.4028234663852886e+38, postInfinityExtrap=RCCE_Constant, preInfinityExtrap=RCCE_Constant, keys []}
      targetRotation=<rot 0 0 0>
     }
     {
      bApplyCustomCurve=false
      bIsHidden=false
      distanceMethod=DefaultMethod
      drivenName=Pose_9
      functionType=DefaultFunction
      targetScale=1
      boneTransforms [
       {targetRotation=<rot -68.9933 -127.6298 -53.9108>, targetTranslation=<4.3575 -0.1466 2.3223>}
       {targetRotation=<rot 33.2732 89.3676 -89.1674>, targetTranslation=<-15.5111 0 0>}
      ]
      customCurve {defaultValue=3.4028234663852886e+38, postInfinityExtrap=RCCE_Constant, preInfinityExtrap=RCCE_Constant, keys []}
      targetRotation=<rot 0 0 0>
     }
    ]
    rBFParams {distanceMethod=Quaternion, radius=100}
    sourceBones [{boneName=clavicle_r}, {boneName=upperarm_r}]
   }
   Save cached pose 'CPFM Cache'(Pose=@n2) {cacheName=CPFM Cache}
   n3 = RigidBody(
    ComponentPose=Local To Component(LocalPose=Use cached pose 'CPFM Cache'())
    ExternalForce=Break Fort Anim Input Skydiving External Force(FortAnimInput_SkydivingExternalForce=Get SkydivingExternalForceInput()) -> (
     bUseSkydivingVectorForce=~
     SkydivingVectorForce=~
     HeadToPelvisDirection=~
     FloatingMultiplier=~
     FloatingAdditive=~
     DivingMultiplier=~
     DivingAdditive=~
     ParachutingMultiplier=~
     ParachutingAdditive=~
    ).SkydivingVectorForce
   ) {
    bClampLinearTranslationLimitToRefPose=true
    bOverrideWorldGravity=true
    overridePhysicsAsset=&/M_MED_Dumpling_Shrimp_Physics
    simulationSpace=BaseBoneSpace
    alphaBoolBlend {blendOption=Linear}
    alphaScaleBiasClamp {bInterpResult=true, interpSpeedDecreasing=4, interpSpeedIncreasing=6}
    baseBoneRef {boneName=root}
    componentAppliedLinearAccClamp=<250 250 250>
    componentLinearAccScale=<1 1 1>
    componentLinearVelScale=<1 1 1>
    overrideWorldGravity {z=-980}
   }
   n4 = Component To Local(ComponentPose=@n3)
   n5 = "Constraint - dyn_jacket_front_04_l (R)"(ComponentPose=Local To Component(LocalPose=Use cached pose 'CPFM Cache'()), Alpha=1, ConstraintWeights_0=1) {
    alphaBoolBlend {blendOption=Linear}
    boneToModify {boneName=dyn_jacket_front_04_l}
    constraintSetup [{transformType=Rotation, targetBone {boneName=calf_l}}]
    constraintWeights [1]
   }
   n6 = "Constraint - dyn_jacket_front_04_r (R)"(ComponentPose=@n5, Alpha=1, ConstraintWeights_0=1) {
    alphaBoolBlend {blendOption=Linear}
    boneToModify {boneName=dyn_jacket_front_04_r}
    constraintSetup [{transformType=Rotation, targetBone {boneName=calf_r}}]
    constraintWeights [1]
   }
   n7 = "Constraint - dyn_jacket_rear_03_l (R)"(ComponentPose=@n6, Alpha=1, ConstraintWeights_0=1) {
    alphaBoolBlend {blendOption=Linear}
    boneToModify {boneName=dyn_jacket_rear_03_l}
    constraintSetup [{transformType=Rotation, targetBone {boneName=calf_l}}]
    constraintWeights [1]
   }
   n8 = "Constraint - dyn_jacket_rear_03_r (R)"(ComponentPose=@n7, Alpha=1, ConstraintWeights_0=1) {
    alphaBoolBlend {blendOption=Linear}
    boneToModify {boneName=dyn_jacket_rear_03_r}
    constraintSetup [{transformType=Rotation, targetBone {boneName=calf_r}}]
    constraintWeights [1]
   }
   n9 = "Constraint - dyn_jacket_side_03_l (R)"(ComponentPose=@n8, Alpha=1, ConstraintWeights_0=1) {
    alphaBoolBlend {blendOption=Linear}
    boneToModify {boneName=dyn_jacket_side_03_l}
    constraintSetup [{transformType=Rotation, targetBone {boneName=calf_l}}]
    constraintWeights [1]
   }
   n10 = "Constraint - dyn_jacket_side_03_r (R)"(ComponentPose=@n9, Alpha=1, ConstraintWeights_0=1) {
    alphaBoolBlend {blendOption=Linear}
    boneToModify {boneName=dyn_jacket_side_03_r}
    constraintSetup [{transformType=Rotation, targetBone {boneName=calf_r}}]
    constraintWeights [1]
   }
   n11 = Component To Local(ComponentPose=@n10)
   n12 = OR Boolean(A=NOT Boolean(A=Get bRigidBodyEnabled()), B="integer > integer"(A=Get CurrentLOD(), B=1) {is=title K2Node_CallFunction})
   n13 = Anim Dynamics - Bone: dyn_pot_front(
    ComponentPose=Local To Component(LocalPose=Blend Poses by bool(BlendPose_0=@n11, BlendPose_1=@n4, bActiveValue=@n12) {blendType=Linear})
    Alpha=1
    ExternalForce="0,0,0"
   ) {
    angularDampingOverride=0.97
    bGravityOverrideInSimSpace=true
    gravityScale=0.65
    lODThreshold=1
    linearDampingOverride=0.97
    numSolverIterationsPostUpdate=4
    numSolverIterationsPreUpdate=16
    alphaBoolBlend {blendOption=Linear}
    boundBone {boneName=dyn_pot_front}
    componentAppliedLinearAccClamp=<10000 10000 10000>
    componentLinearAccScale=<0.5 0.5 0.5>
    componentLinearVelScale=<0.5 0.5 0.5>
    physicsBodyDefinitions [{
      sphereCollisionRadius=0
      boundBone {boneName=dyn_pot_front}
      boxExtents=<5 5 5>
      constraintSetup {angularTargetAxis=AxisZ, twistAxis=AxisZ, angularLimitsMax=<10 15 0>, angularLimitsMin=<-10 -15 0>, angularTarget=<0 0 0>}
      localJointOffset=<0 0 -0.1>
     }]
   }
   n14 = Anim Dynamics - Bone: dyn_pot_back(ComponentPose=@n13, Alpha=1, ExternalForce="0,0,0") {
    angularDampingOverride=0.97
    bGravityOverrideInSimSpace=true
    gravityScale=0.65
    lODThreshold=1
    linearDampingOverride=0.97
    numSolverIterationsPostUpdate=4
    numSolverIterationsPreUpdate=16
    alphaBoolBlend {blendOption=Linear}
    boundBone {boneName=dyn_pot_back}
    componentAppliedLinearAccClamp=<10000 10000 10000>
    componentLinearAccScale=<0.5 0.5 0.5>
    componentLinearVelScale=<0.5 0.5 0.5>
    physicsBodyDefinitions [{
      sphereCollisionRadius=0
      boundBone {boneName=dyn_pot_back}
      boxExtents=<5 5 5>
      constraintSetup {angularTargetAxis=AxisZ, twistAxis=AxisZ, angularLimitsMax=<10 15 0>, angularLimitsMin=<-10 -15 0>, angularTarget=<0 0 0>}
      localJointOffset=<0 0 -0.1>
     }]
   }
   n15 = Get NotCrouching() -> (NotCrouching:double)
   n16 = "Transform (Modify) Bone - Bone: dyn_pot_front"(ComponentPose=@n14, Alpha=@n15, Translation="0,0,0", Rotation="-30,-10,0", Scale="1,1,1") {
    rotationMode=BMM_Additive
    rotationSpace=BCS_ParentBoneSpace
    translationMode=BMM_Additive
    translationSpace=BCS_ParentBoneSpace
    alphaBoolBlend {blendOption=Linear}
    boneToModify {boneName=dyn_pot_front}
    rotation {pitch=-30, yaw=-10}
   }
   n17 = "Transform (Modify) Bone - Bone: dyn_pot_back"(ComponentPose=@n16, Alpha=@n15, Translation="0,0,0", Rotation="-40,-10,0", Scale="1,1,1") {
    rotationMode=BMM_Additive
    rotationSpace=BCS_ParentBoneSpace
    translationMode=BMM_Additive
    translationSpace=BCS_ParentBoneSpace
    alphaBoolBlend {blendOption=Linear}
    boneToModify {boneName=dyn_pot_back}
    rotation {pitch=-40, yaw=-10}
   }
   n18 = Get IsCrouching() -> (IsCrouching:double)
   n19 = "Transform (Modify) Bone - Bone: dyn_pot_front"(ComponentPose=@n17, Alpha=@n18, Translation="0,0,0", Rotation="-80,-10,0", Scale="1,1,1") {
    rotationMode=BMM_Additive
    rotationSpace=BCS_ParentBoneSpace
    translationMode=BMM_Additive
    translationSpace=BCS_ParentBoneSpace
    alphaBoolBlend {blendOption=Linear}
    boneToModify {boneName=dyn_pot_front}
    rotation {pitch=-80, yaw=-10}
   }
   n20 = "Transform (Modify) Bone - Bone: dyn_pot_back"(ComponentPose=@n19, Alpha=@n18, Translation="0,0,0", Rotation="-60,-10,0", Scale="1,1,1") {
    rotationMode=BMM_Additive
    rotationSpace=BCS_ParentBoneSpace
    translationMode=BMM_Additive
    translationSpace=BCS_ParentBoneSpace
    alphaBoolBlend {blendOption=Linear}
    boneToModify {boneName=dyn_pot_back}
    rotation {pitch=-60, yaw=-10}
   }
   n21 = "Constraint - dyn_shoulder_tassle_mstr_r (PPPP)"!(
    ComponentPose=@n20
    Alpha=1
    ConstraintWeights_0=0.1786
    ConstraintWeights_1:float=0.7856
    ConstraintWeights_2:float=0.0233
    ConstraintWeights_3:float=0.0125
   ) -> (Pose) {
    alphaBoolBlend {blendOption=Linear}
    boneToModify {boneName=dyn_shoulder_tassle_mstr_r}
    constraintSetup [
     {targetBone {boneName=upperarm_twist_01_r}}
     {targetBone {boneName=clavicle_r}}
     {targetBone {boneName=dyn_shoulder_tassle_r_r}}
     {targetBone {boneName=dyn_shoulder_tassle_r_l}}
    ]
    constraintWeights [1,1,1,1]
   }
   n22 = "Constraint - dyn_shoulder_tassle_mstr_l (PPPPPPPPPPPP)"!(
    ComponentPose=@n21
    Alpha=1
    ConstraintWeights_0=0.0316
    ConstraintWeights_1:float=0.0234
    ConstraintWeights_2:float=0.7044
    ConstraintWeights_3:float=0.0244
    ConstraintWeights_4:float=0
    ConstraintWeights_5:float=0.0029
    ConstraintWeights_6:float=0
    ConstraintWeights_7:float=0.0025
    ConstraintWeights_8:float=0.0456
    ConstraintWeights_9:float=0.1437
    ConstraintWeights_10:float=0.0121
    ConstraintWeights_11:float=0.0095
   ) -> (Pose) {
    alphaBoolBlend {blendOption=Linear}
    boneToModify {boneName=dyn_shoulder_tassle_mstr_l}
    constraintSetup [
     {targetBone {boneName=spine_03}}
     {targetBone {boneName=dyn_shoulder_tassle_l_l}}
     {targetBone {boneName=clavicle_l}}
     {targetBone {boneName=upperarm_twist_02_l}}
     {targetBone {boneName=spine_01}}
     {targetBone {boneName=spine_02}}
     {targetBone {boneName=thigh_twist_01_l}}
     {targetBone {boneName=spine_04}}
     {targetBone {boneName=spine_05}}
     {targetBone {boneName=upperarm_twist_01_l}}
     {targetBone {boneName=dyn_shoulder_tassle_l_r}}
     {targetBone {boneName=upperarm_l}}
    ]
    constraintWeights [1,1,1,1,1,1,1,1,1,1,1,1]
   }
   n23 = Look At - Bone: dyn_knife(ComponentPose=@n22, Alpha=1) {
    alphaBoolBlend {blendOption=Linear}
    boneToModify {boneName=dyn_knife}
    lookAtLocation=<0 -5 -5>
    lookAtTarget {boneReference {boneName=dyn_jacket_front_03_r}}
    lookAt_Axis {axis {y=0, z=-1}}
   }
   n24 = Copy Bone Delta - Source Bone: dyn_knife - Target Bone: dyn_spatula!(ComponentPose:Component Space Pose Link Structure=@n23, Alpha:float=1) -> (Pose:Component Space Pose Link Structure) {
    is=title AnimGraphNode_CopyBoneDelta
    bCopyRotation=true
    alphaBoolBlend {blendOption=Linear}
    sourceBone {boneName=dyn_knife}
    targetBone {boneName=dyn_spatula}
   }
   n25 = Anim Dynamics - Bone: dyn_spatula(ComponentPose=@n24, Alpha=0.3, ExternalForce="0,0,0") {
    angularDampingOverride=0.97
    bGravityOverrideInSimSpace=true
    bOverrideAngularBias=true
    bOverrideAngularDamping=true
    gravityScale=0.125
    lODThreshold=1
    linearDampingOverride=0.97
    numSolverIterationsPostUpdate=4
    numSolverIterationsPreUpdate=16
    simulationSpace=BoneRelative
    alphaBoolBlend {blendOption=Linear}
    boundBone {boneName=dyn_spatula}
    componentAppliedLinearAccClamp=<10000 10000 10000>
    physicsBodyDefinitions [{
      sphereCollisionRadius=0
      boundBone {boneName=dyn_spatula}
      boxExtents=<10 10 2.5>
      constraintSetup {angularTargetAxis=AxisZ, twistAxis=AxisZ, angularLimitsMax=<0 15 0>, angularLimitsMin=<0 -15 0>, angularTarget=<0 0 0>}
      localJointOffset=<0 0 -1.5>
     }]
    relativeSpaceBone {boneName=root}
   }
   n26 = Anim Dynamics - Bone: dyn_knife(ComponentPose=@n25, Alpha=0.45, ExternalForce="0,0,0") {
    angularDampingOverride=0.97
    bGravityOverrideInSimSpace=true
    bOverrideAngularBias=true
    bOverrideAngularDamping=true
    gravityScale=0.125
    lODThreshold=1
    linearDampingOverride=0.97
    numSolverIterationsPostUpdate=4
    numSolverIterationsPreUpdate=16
    simulationSpace=BoneRelative
    alphaBoolBlend {blendOption=Linear}
    boundBone {boneName=dyn_knife}
    componentAppliedLinearAccClamp=<10000 10000 10000>
    physicsBodyDefinitions [{
      sphereCollisionRadius=0
      boundBone {boneName=dyn_knife}
      boxExtents=<10 10 2.5>
      constraintSetup {angularTargetAxis=AxisZ, twistAxis=AxisZ, angularLimitsMax=<0 15 0>, angularLimitsMin=<0 -15 0>, angularTarget=<0 0 0>}
      localJointOffset=<0 0 -1.5>
     }]
    relativeSpaceBone {boneName=root}
   }
   n27 = SelectFloat(A=0.6, B=0.38, bPickA=Get bIsTargeting())
   n28 = Anim Dynamics - Bone: dyn_shoulder_tassle_l_r(ComponentPose=@n26, Alpha=@n27, ExternalForce="0,0,0") {
    angularDampingOverride=0.97
    angularSpringConstant=9000
    bAngularSpring=true
    bGravityOverrideInSimSpace=true
    bOverrideAngularBias=true
    bOverrideAngularDamping=true
    gravityScale=0.125
    lODThreshold=1
    linearDampingOverride=0.97
    numSolverIterationsPostUpdate=4
    numSolverIterationsPreUpdate=16
    simulationSpace=BoneRelative
    alphaBoolBlend {blendOption=Linear}
    boundBone {boneName=dyn_shoulder_tassle_l_r}
    componentAppliedLinearAccClamp=<10000 10000 10000>
    componentLinearAccScale=<0.5 0.5 0.5>
    componentLinearVelScale=<0.5 0.5 0.5>
    physicsBodyDefinitions [{
      sphereCollisionRadius=0
      boundBone {boneName=dyn_shoulder_tassle_l_r}
      boxExtents=<10 10 2.5>
      constraintSetup {angularTargetAxis=AxisZ, twistAxis=AxisZ, angularLimitsMax=<10 15 30>, angularLimitsMin=<-10 -15 -30>, angularTarget=<0 0 0>}
      localJointOffset=<0 0 -0.5>
     }]
    relativeSpaceBone {boneName=root}
   }
   n29 = Anim Dynamics - Bone: dyn_shoulder_tassle_l_l(ComponentPose=@n28, Alpha=@n27, ExternalForce="0,0,0") {
    angularDampingOverride=0.94
    angularSpringConstant=9000
    bAngularSpring=true
    bGravityOverrideInSimSpace=true
    bOverrideAngularBias=true
    bOverrideAngularDamping=true
    gravityScale=0.125
    lODThreshold=1
    linearDampingOverride=0.94
    numSolverIterationsPostUpdate=4
    numSolverIterationsPreUpdate=16
    simulationSpace=BoneRelative
    alphaBoolBlend {blendOption=Linear}
    boundBone {boneName=dyn_shoulder_tassle_l_l}
    componentAppliedLinearAccClamp=<10000 10000 10000>
    componentLinearAccScale=<0.6 0.6 0.6>
    componentLinearVelScale=<0.6 0.6 0.6>
    physicsBodyDefinitions [{
      sphereCollisionRadius=0
      boundBone {boneName=dyn_shoulder_tassle_l_l}
      boxExtents=<10 10 2.5>
      constraintSetup {angularTargetAxis=AxisZ, twistAxis=AxisZ, angularLimitsMax=<10 15 30>, angularLimitsMin=<-10 -15 -30>, angularTarget=<0 0 0>}
      localJointOffset=<0 0 -1.25>
     }]
    relativeSpaceBone {boneName=root}
   }
   n30 = Anim Dynamics - Bone: dyn_shoulder_tassle_r_r(ComponentPose=@n29, Alpha=@n27, ExternalForce="0,0,0") {
    angularDampingOverride=0.97
    angularSpringConstant=9000
    bAngularSpring=true
    bGravityOverrideInSimSpace=true
    bOverrideAngularBias=true
    bOverrideAngularDamping=true
    gravityScale=0.125
    lODThreshold=1
    linearDampingOverride=0.97
    numSolverIterationsPostUpdate=4
    numSolverIterationsPreUpdate=16
    simulationSpace=BoneRelative
    alphaBoolBlend {blendOption=Linear}
    boundBone {boneName=dyn_shoulder_tassle_r_r}
    componentAppliedLinearAccClamp=<10000 10000 10000>
    componentLinearAccScale=<0.2 0.2 0.2>
    componentLinearVelScale=<0.2 0.2 0.2>
    physicsBodyDefinitions [{
      sphereCollisionRadius=0
      boundBone {boneName=dyn_shoulder_tassle_r_r}
      boxExtents=<10 10 2.5>
      constraintSetup {angularTargetAxis=AxisZ, twistAxis=AxisZ, angularLimitsMax=<10 15 30>, angularLimitsMin=<-10 -15 -30>, angularTarget=<0 0 0>}
      localJointOffset=<0 0 -1>
     }]
    relativeSpaceBone {boneName=root}
   }
   n31 = Anim Dynamics - Bone: dyn_shoulder_tassle_r_l(ComponentPose=@n30, Alpha=@n27, ExternalForce="0,0,0") {
    angularDampingOverride=0.96
    angularSpringConstant=9000
    bAngularSpring=true
    bGravityOverrideInSimSpace=true
    bOverrideAngularBias=true
    bOverrideAngularDamping=true
    gravityScale=0.125
    lODThreshold=1
    linearDampingOverride=0.96
    numSolverIterationsPostUpdate=4
    numSolverIterationsPreUpdate=16
    simulationSpace=BoneRelative
    alphaBoolBlend {blendOption=Linear}
    boundBone {boneName=dyn_shoulder_tassle_r_l}
    componentAppliedLinearAccClamp=<10000 10000 10000>
    componentLinearAccScale=<0.3 0.3 0.3>
    componentLinearVelScale=<0.3 0.3 0.3>
    physicsBodyDefinitions [{
      sphereCollisionRadius=0
      boundBone {boneName=dyn_shoulder_tassle_r_l}
      boxExtents=<10 10 2.5>
      constraintSetup {angularTargetAxis=AxisZ, twistAxis=AxisZ, angularLimitsMax=<10 15 30>, angularLimitsMin=<-10 -15 -30>, angularTarget=<0 0 0>}
      localJointOffset=<0 0 -0.75>
     }]
    relativeSpaceBone {boneName=root}
   }
   n32 = "Transform (Modify) Bone - Bone: dyn_strap_corrective_r"!(ComponentPose=@n31, Alpha=0, Translation="5,0,0", Rotation="0,0,0") -> (Pose) {
    rotationMode=BMM_Replace
    rotationSpace=BCS_BoneSpace
    translationMode=BMM_Additive
    translationSpace=BCS_BoneSpace
    alphaBoolBlend {blendOption=Linear}
    boneToModify {boneName=dyn_strap_corrective_r}
    translation {x=5}
   }
   Save cached pose 'dynCache'(Pose=Component To Local(ComponentPose=@n32)) {cacheName=dynCache}
  ]}
 EventGraph {nodes [
   Event BlueprintUpdateAnimation()
   > n1 = Sequence!(execute) -> (then_0, then_1, then_2:Exec)
   n2 = Get bIsCrouching()
   n3 = Get bIsCrouchSprinting()
   Set NotCrouching(execute=@n1.then_1, NotCrouching:double=SelectFloat(A=0, B=1, bPickA=OR Boolean(A=@n2, B=@n3))) -> (Output_Get:double)
   n4 = Get bIsCrouching()
   n5 = Get bIsCrouchSprinting()
   Set IsCrouching(execute=@n1.then_2, IsCrouching:double=SelectFloat(A=1, B=0, bPickA=OR Boolean(A=@n4, B=@n5))) -> (Output_Get:double)
   Event OnParachuteOpened()
   > n6 = Delay(Duration=1.5)
   ResetDynamics(execute=[@n6.then, @n7.then, @n8.then, @n11.then], InTeleportType=TeleportPhysics)
   > ResetDynamics(InTeleportType=ResetPhysics)
   n7 = Event OnBeginSkydiving()
   n8 = Event OnEndSkydiving()
   n9 = Event OnEndCrouching()
   n10 = Get bIsCrouching()
   n11 = Branch(execute=[@n9.then, @n12.then], Condition=AND Boolean(A=Get bIsTargeting(), B=NOT Boolean(A=@n10)))
   n12 = Event OnBeginTargeting() {is=title K2Node_Event}
  ]}
]
variables [
 IsCrouching {type=number}
 NotCrouching {type=number}
]
