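# Physics asset for the face accessory of the male medium Dumpling Man Shrimp body cosmetic, giving physics bodies to its left and right bandana strands.
# Each strand is a nine-bone chain of identical boxes, with its first bone constrained to a sphere body on the head.
# Tight swing and twist limits restrict every joint, and the wider swing axis switches from swing2 to swing1 at the fifth bone of each chain.
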PhysicsAsset /BRCosmetics/Characters/Player/Male/Medium/Bodies/M_MED_Dumpling_Man_Shrimp/Meshes/M_MED_Dumpling_Man_Shrimp_FaceAcc_Physics
bodies {
 dyn_bandana_l_01 [{shapeType=Box, extentX=3, extentY=5, extentZ=3, center=<2 0 0>}]
 dyn_bandana_l_02 [{shapeType=Box, extentX=3, extentY=5, extentZ=3, center=<2 0 0>}]
 dyn_bandana_l_03 [{shapeType=Box, extentX=3, extentY=5, extentZ=3, center=<2 0 0>}]
 dyn_bandana_l_04 [{shapeType=Box, extentX=3, extentY=5, extentZ=3, center=<2 0 0>}]
 dyn_bandana_l_05 [{shapeType=Box, extentX=3, extentY=5, extentZ=3, center=<2 0 0>}]
 dyn_bandana_l_06 [{shapeType=Box, extentX=3, extentY=5, extentZ=3, center=<2 0 0>}]
 dyn_bandana_l_07 [{shapeType=Box, extentX=3, extentY=5, extentZ=3, center=<2 0 0>}]
 dyn_bandana_l_08 [{shapeType=Box, extentX=3, extentY=5, extentZ=3, center=<2 0 0>}]
 dyn_bandana_l_09 [{shapeType=Box, extentX=3, extentY=5, extentZ=3, center=<2 0 0>}]
 dyn_bandana_r_01 [{shapeType=Box, extentX=3, extentY=5, extentZ=3, center=<2 0 0>}]
 dyn_bandana_r_02 [{shapeType=Box, extentX=3, extentY=5, extentZ=3, center=<2 0 0>}]
 dyn_bandana_r_03 [{shapeType=Box, extentX=3, extentY=5, extentZ=3, center=<2 0 0>}]
 dyn_bandana_r_04 [{shapeType=Box, extentX=3, extentY=5, extentZ=3, center=<2 0 0>}]
 dyn_bandana_r_05 [{shapeType=Box, extentX=3, extentY=5, extentZ=3, center=<2 0 0>}]
 dyn_bandana_r_06 [{shapeType=Box, extentX=3, extentY=5, extentZ=3, center=<2 0 0>}]
 dyn_bandana_r_07 [{shapeType=Box, extentX=3, extentY=5, extentZ=3, center=<2 0 0>}]
 dyn_bandana_r_08 [{shapeType=Box, extentX=3, extentY=5, extentZ=3, center=<2 0 0>}]
 dyn_bandana_r_09 [{shapeType=Box, extentX=3, extentY=5, extentZ=3, center=<2 0 0>}]
 head [{shapeType=Sphere, radius=16.25, center=<3.5122 -0.717 -0.0001>}]
}
constraints [
 {bone1Name=dyn_bandana_r_01, bone2Name=head, swing1LimitDegrees=5, swing2LimitDegrees=15, twistLimitDegrees=5}
 {bone1Name=dyn_bandana_r_02, bone2Name=dyn_bandana_r_01, swing1LimitDegrees=5, swing2LimitDegrees=15, twistLimitDegrees=5}
 {bone1Name=dyn_bandana_r_03, bone2Name=dyn_bandana_r_02, swing1LimitDegrees=5, swing2LimitDegrees=15, twistLimitDegrees=5}
 {bone1Name=dyn_bandana_r_04, bone2Name=dyn_bandana_r_03, swing1LimitDegrees=5, swing2LimitDegrees=15, twistLimitDegrees=5}
 {bone1Name=dyn_bandana_r_05, bone2Name=dyn_bandana_r_04, swing1LimitDegrees=15, swing2LimitDegrees=5, twistLimitDegrees=5}
 {bone1Name=dyn_bandana_r_06, bone2Name=dyn_bandana_r_05, swing1LimitDegrees=15, swing2LimitDegrees=5, twistLimitDegrees=5}
 {bone1Name=dyn_bandana_r_07, bone2Name=dyn_bandana_r_06, swing1LimitDegrees=15, swing2LimitDegrees=5, twistLimitDegrees=5}
 {bone1Name=dyn_bandana_r_08, bone2Name=dyn_bandana_r_07, swing1LimitDegrees=15, swing2LimitDegrees=5, twistLimitDegrees=5}
 {bone1Name=dyn_bandana_r_09, bone2Name=dyn_bandana_r_08, swing1LimitDegrees=15, swing2LimitDegrees=5, twistLimitDegrees=5}
 {bone1Name=dyn_bandana_l_01, bone2Name=head, swing1LimitDegrees=5, swing2LimitDegrees=15, twistLimitDegrees=5}
 {bone1Name=dyn_bandana_l_02, bone2Name=dyn_bandana_l_01, swing1LimitDegrees=5, swing2LimitDegrees=15, twistLimitDegrees=5}
 {bone1Name=dyn_bandana_l_03, bone2Name=dyn_bandana_l_02, swing1LimitDegrees=5, swing2LimitDegrees=15, twistLimitDegrees=5}
 {bone1Name=dyn_bandana_l_04, bone2Name=dyn_bandana_l_03, swing1LimitDegrees=5, swing2LimitDegrees=15, twistLimitDegrees=5}
 {bone1Name=dyn_bandana_l_05, bone2Name=dyn_bandana_l_04, swing1LimitDegrees=15, swing2LimitDegrees=5, twistLimitDegrees=5}
 {bone1Name=dyn_bandana_l_06, bone2Name=dyn_bandana_l_05, swing1LimitDegrees=15, swing2LimitDegrees=5, twistLimitDegrees=5}
 {bone1Name=dyn_bandana_l_07, bone2Name=dyn_bandana_l_06, swing1LimitDegrees=15, swing2LimitDegrees=5, twistLimitDegrees=5}
 {bone1Name=dyn_bandana_l_08, bone2Name=dyn_bandana_l_07, swing1LimitDegrees=15, swing2LimitDegrees=5, twistLimitDegrees=5}
 {bone1Name=dyn_bandana_l_09, bone2Name=dyn_bandana_l_08, swing1LimitDegrees=15, swing2LimitDegrees=5, twistLimitDegrees=5}
]
